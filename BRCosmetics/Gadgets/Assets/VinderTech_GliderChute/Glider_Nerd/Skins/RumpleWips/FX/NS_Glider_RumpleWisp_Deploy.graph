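# Niagara deploy effect for the RumpleWisp glider skin that throws a burst of frost, snow and snowflakes from the glider's sockets.
# FrostBurst spawns faint SubUV-animated puffs drifting outward, while SnowBurst and SnowflakeBurst push small sprites from an origin below and stir them with curl noise.
# All three emitters tint by time of day and adjust velocity for front end versus gameplay.
# SnowBurst adds a second sprite renderer, alongside computed glow size and color.
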NiagaraSystem /BRCosmetics/Gadgets/Assets/VinderTech_GliderChute/Glider_Nerd/Skins/RumpleWips/FX/NS_Glider_RumpleWisp_Deploy
dependencies {
 usedDataInterfaces [&NiagaraDataInterfaceCurve, &NiagaraDataInterfaceSocketReader]
 usedDynamicInputs [&Vector2DFromFloat, &MakeVector2D]
 usedModules [
  &SystemState
  &self:SystemSpawnScript.SystemScriptSource.SystemScriptGraph.NiagaraNodeAssignment_2.SetVariables_#1
  &self:SystemSpawnScript.SystemScriptSource.SystemScriptGraph.NiagaraNodeAssignment_3.SetVariables_#2
 ]
}
emitters {
 FrostBurst {
  emitterSpawnScript [/SetVariables(Emitter.VelOrigin:NiagaraPosition=<0 0 -50>) {isSetParameters=true}]
  emitterUpdateScript [
   &EmitterState(
    Inactive Response=::NewEnumerator1
    Life Cycle Mode=::NewEnumerator1
    Loop Behavior=::NewEnumerator1
    Loop Delay=0.0625
    Loop Duration=1
    UseLoopDelay=true
   )
   &SpawnBurst_Instantaneous(Spawn Count=24, Use Spawn Probability=false, UseLoopCountLimit=false)
  ]
  particleSpawnScript [
   &InitializeParticle(
    Color=<rgba 0.8799 0.9947 1 0.025>
    Lifetime=&RandomRangeFloat(Minimum=0.3, Maximum=1)
    Sprite Rotation Mode=::NewEnumerator2
    Sprite Size Mode=::NewEnumerator1
    Sprite UV Mode=::NewEnumerator3
    Uniform Sprite Size Max=48
    Uniform Sprite Size Min=32
    UsePositionOffset=false
    Write Material Random=true
   )
   &FN_SetSpawnGroup()
   &SocketLocation(
    SamplePosOffset=true
    Sampled Position Offset=&RandomRangeVector(Minimum=<-15 -50 0>, Maximum=<15 50 0>)
    Socket Index=&SpawnGroup()
    Socket Reader=$System.Socket Reader
    Socket Sampling Mode=::NewEnumerator3
    VelocitySpeedLimit=false
   )
   &AddVelocity(
    Velocity=&NormalizeVector(
     Vector To Normalize=&Multiply_Vector(A=Linked() {linkedVariable {name=Particles.Position, type=NiagaraPosition}}, B=<1 1 0>)
     Safe Normalize=true
     Safe Fallback Threshold=0
     Safe Fallback Vector=<0 0 0.001>
    )
    Velocity Speed Scale=&RandomRangeFloat(Minimum=50, Maximum=80)
   ) {module=AddVelocity002}
   &AddVelocity(
    Rotation Coordinate Space=::World
    Velocity=&RandomRangeVector(Minimum=<0 0 20>, Maximum=<0 0 40>)
    Velocity Speed Scale=&FN_FrontEndFloatModifier(Gameplay=0, Front End=1)
   ) {module=AddVelocity001}
   &CameraOffset(Camera Offset Amount=5)
  ]
  particleUpdateScript [
   &ParticleState()
   &SolveForcesAndVelocity(Clamp Velocity=false, Limit Acceleration=false)
   /Niagara/Modules/Update/SubUV/SubUVAnimation(Number Of Frames=63, SubUV Lookup Index=&Multiply_Float(A=$Particles.NormalizedAge, B=0.34))
   &ScaleSpriteSize(
    Uniform Curve Scale=2
    Uniform Curve Sprite Scale=DataInterface() {propertyValues {Curve {keys [<at 0 0.5 RCIM_Cubic>, {tangentMode=RCTM_User, time=1, value=1}]}}}
   )
   &DynamicMaterialParameters(
    Index 0 Param 1=0.1
    Module.Index0Float0Write=true
    Module.Index0Float1Write=true
    Module.Index0Float2Write=true
    Module.Index0Float3Write=true
   )
   &ScaleColor(
    Scale Alpha=&FloatFromCurve(FloatCurve=DataInterface() {propertyValues {Curve {keys [
         {arriveTangent=0.9664, interpMode=RCIM_Cubic, leaveTangent=0.9664, tangentMode=RCTM_User, time=0, value=0}
         <at 0.429 1 RCIM_Cubic>
         {arriveTangent=-0.669, interpMode=RCIM_Cubic, leaveTangent=-0.669, tangentMode=RCTM_User, time=1, value=0}
        ]}}})
    ScaleA=true
    ScaleRGB=true
   )
   &ScaleColor(Color Scale:LinearColor=$System.TOD Color, Scale Mode=3) {module=ScaleColor001}
  ]
  rendererClasses [&NiagaraSpriteRendererProperties]
  renderers [{rendererClass=&NiagaraSpriteRendererProperties}]
 }
 SnowBurst {
  emitterSpawnScript [/SetVariables(Emitter.VelOrigin:NiagaraPosition=<0 0 -50>) {isSetParameters=true, module=SetVariables_#3}]
  emitterUpdateScript [
   &EmitterState(
    Inactive Response=::NewEnumerator1
    Life Cycle Mode=::NewEnumerator1
    Loop Behavior=::NewEnumerator1
    Loop Delay=0.0625
    Loop Duration=1
    UseLoopDelay=true
   )
   &SpawnBurst_Instantaneous(Spawn Count=50, Use Spawn Probability=false, UseLoopCountLimit=false)
  ]
  particleSpawnScript [
   &InitializeParticle(
    Color=&MakeLinearColorFromVectorAndFloat(
     "Vector (RGB)"=&MakeVectorFromLinearColorRGB(Color=<rgba 1 1 1 1>)
     "Float (Alpha)"=&RandomRangeFloat(Minimum=0.1, Maximum=0.2)
    )
    Lifetime=&RandomRangeFloat(Minimum=0.2, Maximum=1.2)
    Sprite Rotation Mode=::NewEnumerator2
    Sprite Size Mode=::NewEnumerator3
    Uniform Sprite Size=&FloatFromCurve(
     FloatCurve=DataInterface() {propertyValues {Curve {keys [<at 0 1>,<at 0.8 2 RCIM_Constant>,<at 0.95 3>]}}}
     CurveIndex=&RandomRangeFloat(Minimum=0, Maximum=1)
    )
    UsePositionOffset=false
    Write Material Random=true
   )
   &SocketLocation(
    SamplePosOffset=true
    Sampled Position Offset=&RandomRangeVector(Minimum=<-50 -5 0>, Maximum=<50 5 0>)
    Socket Index=&RandomRangeInt(Minimum=0, Maximum=3)
    Socket Reader=$System.Socket Reader
    Socket Sampling Mode=::NewEnumerator3
    VelocitySpeedLimit=false
   )
   &AddVelocity(
    Module.Use Falloff Distance=false
    Override Seed=false
    Velocity Mode=::NewEnumerator1
    Velocity Origin=$Emitter.VelOrigin
    Velocity Speed=&RandomRangeFloat(Minimum=15, Maximum=120)
   ) {module=AddVelocity002}
   &AddVelocity(
    Velocity=&Add_Vector(
     A=&RandomVector(Vector Scale=25)
     B=&FN_FrontEndVector3Modifier(Gameplay=<0 300 50>, Front End=<0 400 200>)
    )
    Velocity Speed Scale=&RandomRangeFloat(Minimum=0.1, Maximum=1)
   )
   &CameraOffset(Camera Offset Amount=5)
  ]
  particleUpdateScript [
   &ParticleState()
   &CurlNoiseForce(
    Bias Noise=false
    Noise Strength=&FloatFromCurve(
     FloatCurve=DataInterface() {propertyValues {Curve {keys [{interpMode=RCIM_Cubic, tangentMode=RCTM_User, time=0, value=0.2}, {interpMode=RCIM_Cubic, tangentMode=RCTM_User, time=1, value=1}]}}}
     Scale Curve=350
    )
    Pan Noise=false
    Random Seed=33
   ) {module=CurlNoiseForce001}
   &SolveForcesAndVelocity(Clamp Velocity=false, Limit Acceleration=false)
   &ScaleSpriteSize(
    Uniform Curve Scale=&RandomRangeFloat(Minimum=1, Maximum=2.5)
    Uniform Curve Sprite Scale=DataInterface() {propertyValues {Curve {keys [
        {tangentMode=RCTM_User, time=0, value=1}
        {arriveTangent=-0.0405, interpMode=RCIM_Cubic, leaveTangent=-0.0405, tangentMode=RCTM_User, time=0.7009, value=1}
        {arriveTangent=-8.0089, interpMode=RCIM_Cubic, leaveTangent=-8.0089, tangentMode=RCTM_User, time=1, value=0}
       ]}}}
   )
   &ScaleColor(Color Scale:LinearColor=$System.TOD Color, Scale Mode=3)
   /SetVariables(Particles.SpriteSizeGlow:Vector2f=&Multiply_Vector2D_ByFloat(Float=1.2, Vector2D=$Particles.SpriteSize)) {isSetParameters=true, module=SetVariables_#4}
   /SetVariables(Particles.Glow Color=&MultiplyLinearColor(Linear Color=$Particles.Color, Scale Factor=<2 2 2 0.15>)) {isSetParameters=true, module=SetVariables_#5}
  ]
  rendererClasses [&NiagaraSpriteRendererProperties]
  renderers [
   {rendererClass=&NiagaraSpriteRendererProperties}
   {rendererClass=&NiagaraSpriteRendererProperties, rendererIndex=1}
  ]
 }
 SnowflakeBurst {
  emitterSpawnScript [/SetVariables(Emitter.VelOrigin:NiagaraPosition=<0 0 -50>) {isSetParameters=true, module=SetVariables_#3}]
  emitterUpdateScript [
   &EmitterState(
    Inactive Response=::NewEnumerator1
    Life Cycle Mode=::NewEnumerator1
    Loop Behavior=::NewEnumerator1
    Loop Delay=0.0625
    Loop Duration=1
    UseLoopDelay=true
   )
   &SpawnBurst_Instantaneous(Spawn Count=13, Use Spawn Probability=false, UseLoopCountLimit=false)
  ]
  particleSpawnScript [
   &InitializeParticle(
    Color=<rgba 0.7478 0.8882 1 1>
    Lifetime=&RandomRangeFloat(Minimum=0.2, Maximum=0.4)
    Sprite Rotation Mode=::NewEnumerator2
    Sprite Size Mode=::NewEnumerator1
    Uniform Sprite Size Max=8
    Uniform Sprite Size Min=4
    UsePositionOffset=false
    Write Material Random=true
   )
   &SocketLocation(
    SamplePosOffset=true
    Sampled Position Offset=&RandomRangeVector(Minimum=<-35 -15 0>, Maximum=<35 15 0>)
    Socket Index=&RandomRangeInt(Minimum=0, Maximum=3)
    Socket Reader=$System.Socket Reader
    Socket Sampling Mode=::NewEnumerator3
    VelocitySpeedLimit=false
   )
   &AddVelocity(
    Module.Use Falloff Distance=false
    Override Seed=false
    Velocity Mode=::NewEnumerator1
    Velocity Origin=$Emitter.VelOrigin
    Velocity Speed=&RandomRangeFloat(Minimum=5, Maximum=120)
   ) {module=AddVelocity002}
   &AddVelocity(
    Velocity=&Add_Vector(
     A=&RandomVector(Vector Scale=25)
     B=&FN_FrontEndVector3Modifier(Gameplay=<0 0 15>, Front End=<0 0 40>)
    )
    Velocity Speed Scale=&RandomRangeFloat(Minimum=0.1, Maximum=1)
   )
   &CameraOffset(Camera Offset Amount=5)
  ]
  particleUpdateScript [
   &ParticleState()
   &CurlNoiseForce(
    Bias Noise=false
    Noise Strength=&FloatFromCurve(
     FloatCurve=DataInterface() {propertyValues {Curve {keys [{interpMode=RCIM_Cubic, tangentMode=RCTM_User, time=0, value=0.2}, {interpMode=RCIM_Cubic, tangentMode=RCTM_User, time=1, value=1}]}}}
     Scale Curve=350
    )
    Pan Noise=false
    Random Seed=33
   ) {module=CurlNoiseForce001}
   &SolveForcesAndVelocity(Clamp Velocity=false, Limit Acceleration=false)
   &ScaleSpriteSize(
    Uniform Curve Scale=&RandomRangeFloat(Minimum=1, Maximum=2.5)
    Uniform Curve Sprite Scale=DataInterface() {propertyValues {Curve {keys [
        {tangentMode=RCTM_User, time=0, value=1}
        {arriveTangent=-0.0405, interpMode=RCIM_Cubic, leaveTangent=-0.0405, tangentMode=RCTM_User, time=0.7009, value=1}
        {arriveTangent=-8.0089, interpMode=RCIM_Cubic, leaveTangent=-8.0089, tangentMode=RCTM_User, time=1, value=0}
       ]}}}
   )
   &ScaleColor(
    Color Scale:LinearColor=&FNTimeOfDayColor(
     Night Color=<rgba 0.72 0.8507 1 0.6>
     Morning Color=<rgba 0.9217 0.7865 0.691 0.56>
     Day Color=<rgba 1 1 1 1>
     Evening Color=<rgba 0.9217 0.6535 0.4649 0.56>
    )
    Scale Mode=3
   )
   /SetVariables(Particles.SpriteSizeGlow:Vector2f=&Multiply_Vector2D_ByFloat(Float=1.2, Vector2D=$Particles.SpriteSize)) {isSetParameters=true, module=SetVariables_#4}
   /SetVariables(Particles.Glow Color=&MultiplyLinearColor(Linear Color=$Particles.Color, Scale Factor=<2 2 2 0.15>)) {isSetParameters=true, module=SetVariables_#5}
  ]
  rendererClasses [&NiagaraSpriteRendererProperties]
  renderers [{rendererClass=&NiagaraSpriteRendererProperties}]
 }
}
summary {
 systemName=NS_Glider_RumpleWisp_Deploy
 emitters [
  {emitterName=FrostBurst, rendererClasses [&NiagaraSpriteRendererProperties]}
  {emitterName=SnowBurst, rendererClasses [&NiagaraSpriteRendererProperties]}
  {emitterName=SnowflakeBurst, rendererClasses [&NiagaraSpriteRendererProperties]}
 ]
 userVariables [
  User.bIsFrontEnd {type=NiagaraBool}
  User.bIsFrontEndPreview {type=NiagaraBool}
  User.bIsFullyDeployed {type=NiagaraBool}
 ]
}
